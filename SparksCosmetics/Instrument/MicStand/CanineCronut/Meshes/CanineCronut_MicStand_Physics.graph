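# Physics asset for the CanineCronut mic stand cosmetic in Sparks, giving its dangling ropes, cords, leaves and side boxes secondary motion.
# A box body on the handle anchors five-bone top and bottom rope chains, three-bone left and right cord chains, and single-bone leaves and boxes.
# Capsules and boxes form the bodies, with compound shapes on each rope's fourth bone.
# All constraints lock twist and restrict swing, some locking one swing axis.
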PhysicsAsset /SparksCosmetics/Instrument/MicStand/CanineCronut/Meshes/CanineCronut_MicStand_Physics
bodies {
 dyn_bottom_rope_01 [{shapeType=Capsule, length=12.0676, radius=4.3882, shapeName=dyn_bottom_rope_01_capsule, center=<2 0 0>, rotation=<rot 0 -3.7377 -81.7633>}]
 dyn_bottom_rope_02 [{
   shapeType=Capsule
   length=19.8069
   radius=2.8164
   shapeName=dyn_bottom_rope_02_capsule
   center=<2 0 0>
   rotation=<rot 0.058 -7.8623 -88.7823>
  }]
 dyn_bottom_rope_03 [{shapeType=Capsule, length=14.4856, radius=3.4444, shapeName=dyn_bottom_rope_03_capsule, center=<4 -1 0>, rotation=<rot 0 168.3955 -89.4817>}]
 dyn_bottom_rope_04 [
  {
   shapeType=Capsule
   length=12.0092
   radius=1.0605
   shapeName=dyn_bottom_rope_04_capsule_R
   center=<0.1998 8 -0.1821>
   rotation=<rot -90 -11.2457 0>
  }
  {
   shapeType=Capsule
   length=12.8924
   radius=0.932
   shapeName=dyn_bottom_rope_04_capsule_L
   center=<-1 -8 -0.7271>
   rotation=<rot -90 18.0779 0>
  }
  {
   shapeType=Box
   extentX=5
   extentY=1.3959
   extentZ=12.5871
   shapeName=dyn_bottom_rope_04_box
   center=<2 0 0.5>
   rotation=<rot 4.1226 178.1903 -88.3854>
  }
 ]
 dyn_bottom_rope_05 [{
   shapeType=Box
   extentX=5.3995
   extentY=1.7872
   extentZ=11.5074
   shapeName=dyn_bottom_rope_05_box
   center=<1.5 0 0>
   rotation=<rot 5.4519 -178.0524 -84.2083>
  }]
 dyn_box_l [{shapeType=Box, extentX=11.1227, extentY=1.4221, extentZ=20.3073, shapeName=dyn_box_l_box, center=<5.2442 0 0.4402>, rotation=<rot 0 180 -90>}]
 dyn_box_r [{shapeType=Box, extentX=11.1107, extentY=0.567, extentZ=20.1437, shapeName=dyn_box_r_box, center=<5 0 2>, rotation=<rot 0 0.0001 -90>}]
 dyn_cord_l_01 [{
   shapeType=Capsule
   length=2.1047
   radius=0.7591
   shapeName=dyn_cord_l_01_capsule
   center=<1.4638 -0.0364 0.5782>
   rotation=<rot 2.8134 89.3186 -69.1802>
  }]
 dyn_cord_l_02 [{
   shapeType=Capsule
   length=4.6417
   radius=0.943
   shapeName=dyn_cord_l_02_capsule
   center=<2.7867 -0.0168 0.4049>
   rotation=<rot 0.7263 89.9594 -84.8735>
  }]
 dyn_cord_l_03 [{
   shapeType=Capsule
   length=5.5245
   radius=1.2489
   shapeName=dyn_cord_l_03_capsule
   center=<5.146 -0.3094 -0.5277>
   rotation=<rot 0 -94.5078 -84.6027>
  }]
 dyn_cord_r_01 [{
   shapeType=Capsule
   length=1.4813
   radius=0.956
   shapeName=dyn_cord_r_01_capsule
   center=<0.6469 0.1013 0.3717>
   rotation=<rot 0 76.2104 -71.2304>
  }]
 dyn_cord_r_02 [{
   shapeType=Capsule
   length=3.1886
   radius=0.951
   shapeName=dyn_cord_r_02_capsule
   center=<1.6955 -0.1398 0.6184>
   rotation=<rot -3.2504 88.545 -67.9767>
  }]
 dyn_cord_r_03 [{
   shapeType=Capsule
   length=5.4296
   radius=1.2347
   shapeName=dyn_cord_r_03_capsule
   center=<5.3467 0.11 0.0151>
   rotation=<rot 0 -88.0036 -87.6631>
  }]
 dyn_leaf_bottom_rope_l [{
   shapeType=Capsule
   length=10.1813
   radius=0.9818
   shapeName=dyn_leaf_bottom_rope_l_capsule
   center=<9 -1.5 -2>
   rotation=<rot 0 -78.5321 -72.6061>
  }]
 dyn_leaf_bottom_rope_r [{
   shapeType=Capsule
   length=11.2863
   radius=1.1851
   shapeName=dyn_leaf_bottom_rope_r_capsule
   center=<7 0 0>
   rotation=<rot 0 -100.5419 -78.8885>
  }]
 dyn_leaf_top_rope_l [{shapeType=Capsule, length=10.0969, radius=0.9763, shapeName=dyn_leaf_top_rope_l_capsule, center=<8 -2 0>, rotation=<rot 0 -89.8449 -87.1551>}]
 dyn_leaf_top_rope_r [{shapeType=Capsule, length=8.0332, radius=1.0362, shapeName=dyn_leaf_top_rope_r_capsule, center=<8 2 1.5>, rotation=<rot 0 82.6583 -83.3585>}]
 dyn_top_rope_01 [{shapeType=Capsule, length=10, radius=4, shapeName=dyn_top_rope_01_capsule, rotation=<rot 0 -10.2359 90>}]
 dyn_top_rope_02 [{
   shapeType=Capsule
   length=17.5143
   radius=4.8812
   shapeName=dyn_top_rope_02_capsule
   center=<0 0.75 5>
   rotation=<rot -0.27 -5.1713 -81.3443>
  }]
 dyn_top_rope_03 [{shapeType=Capsule, length=16.1294, radius=4.0223, shapeName=dyn_top_rope_03_capsule, center=<0 0 3.7>, rotation=<rot 0.0001 -8.1324 90>}]
 dyn_top_rope_04 [
  {
   shapeType=Capsule
   length=10
   radius=1
   shapeName=dyn_top_rope_04_capsule_R
   center=<-1.0936 9.5636 -0.6986>
   rotation=<rot -81.0789 -10.9183 0>
  }
  {
   shapeType=Capsule
   length=9.234
   radius=1.3111
   shapeName=dyn_top_rope_04_capsule_L
   center=<-2.8451 -11.3105 0.3898>
   rotation=<rot -86.3112 -56.6701 68.2029>
  }
  {
   shapeType=Box
   extentX=6.1036
   extentY=1.3576
   extentZ=18.0794
   shapeName=dyn_top_rope_04_box
   center=<3 -1 1.7>
   rotation=<rot 1.1007 171.6625 -89.6143>
  }
 ]
 dyn_top_rope_05 [{
   shapeType=Capsule
   length=4.9408
   radius=3.4451
   shapeName=dyn_top_rope_05_capsule
   center=<1.5 0 2.5>
   rotation=<rot 0.4169 171.4413 -83.5594>
  }]
 handle [{shapeType=Box, extentX=18.861, extentY=19.6329, extentZ=46.2873, shapeName=handle_box, center=<0.0075 0.0908 -36.1085>}]
}
constraints [
 {bone1Name=dyn_box_r, bone2Name=handle, swing1LimitDegrees=0, swing2LimitDegrees=30, twistLimitDegrees=45, twistMotion=Locked}
 {
  bone1Name=dyn_box_l
  bone2Name=handle
  swing1LimitDegrees=0
  swing1Motion=Locked
  swing2LimitDegrees=30
  twistLimitDegrees=45
  twistMotion=Locked
 }
 {
  bone1Name=dyn_cord_l_01
  bone2Name=handle
  swing1LimitDegrees=10
  swing2LimitDegrees=10
  swing2Motion=Locked
  twistLimitDegrees=45
  twistMotion=Locked
 }
 {bone1Name=dyn_cord_l_02, bone2Name=dyn_cord_l_01, swing1LimitDegrees=10, swing2LimitDegrees=35, twistLimitDegrees=45, twistMotion=Locked}
 {bone1Name=dyn_cord_l_03, bone2Name=dyn_cord_l_02, swing1LimitDegrees=10, swing2LimitDegrees=10, twistLimitDegrees=45, twistMotion=Locked}
 {
  bone1Name=dyn_cord_r_01
  bone2Name=handle
  swing1LimitDegrees=10
  swing2LimitDegrees=5
  swing2Motion=Locked
  twistLimitDegrees=45
  twistMotion=Locked
 }
 {bone1Name=dyn_cord_r_02, bone2Name=dyn_cord_r_01, swing1LimitDegrees=10, swing2LimitDegrees=35, twistLimitDegrees=45, twistMotion=Locked}
 {bone1Name=dyn_cord_r_03, bone2Name=dyn_cord_r_02, swing1LimitDegrees=20, swing2LimitDegrees=10, twistLimitDegrees=45, twistMotion=Locked}
 {bone1Name=dyn_top_rope_01, bone2Name=handle, swing1LimitDegrees=10, swing2LimitDegrees=20, twistLimitDegrees=45, twistMotion=Locked}
 {bone1Name=dyn_top_rope_02, bone2Name=dyn_top_rope_01, swing1LimitDegrees=40, swing2LimitDegrees=45, twistLimitDegrees=45, twistMotion=Locked}
 {bone1Name=dyn_top_rope_03, bone2Name=dyn_top_rope_02, swing1LimitDegrees=70, swing2LimitDegrees=45, twistLimitDegrees=45, twistMotion=Locked}
 {bone1Name=dyn_top_rope_04, bone2Name=dyn_top_rope_03, swing1LimitDegrees=65, swing2LimitDegrees=45, twistLimitDegrees=45, twistMotion=Locked}
 {bone1Name=dyn_top_rope_05, bone2Name=dyn_top_rope_04, swing1LimitDegrees=60, swing2LimitDegrees=20, twistLimitDegrees=45, twistMotion=Locked}
 {bone1Name=dyn_bottom_rope_01, bone2Name=handle, swing1LimitDegrees=10, swing2LimitDegrees=15, twistLimitDegrees=45, twistMotion=Locked}
 {
  bone1Name=dyn_bottom_rope_02
  bone2Name=dyn_bottom_rope_01
  swing1LimitDegrees=20
  swing2LimitDegrees=10
  twistLimitDegrees=45
  twistMotion=Locked
 }
 {
  bone1Name=dyn_bottom_rope_03
  bone2Name=dyn_bottom_rope_02
  swing1LimitDegrees=25
  swing2LimitDegrees=25
  twistLimitDegrees=45
  twistMotion=Locked
 }
 {
  bone1Name=dyn_bottom_rope_04
  bone2Name=dyn_bottom_rope_03
  swing1LimitDegrees=30
  swing2LimitDegrees=20
  twistLimitDegrees=45
  twistMotion=Locked
 }
 {
  bone1Name=dyn_bottom_rope_05
  bone2Name=dyn_bottom_rope_04
  swing1LimitDegrees=15
  swing2LimitDegrees=30
  twistLimitDegrees=45
  twistMotion=Locked
 }
 {
  bone1Name=dyn_leaf_bottom_rope_l
  bone2Name=handle
  swing1LimitDegrees=40
  swing2LimitDegrees=35
  swing2Motion=Locked
  twistLimitDegrees=45
  twistMotion=Locked
 }
 {
  bone1Name=dyn_leaf_bottom_rope_r
  bone2Name=handle
  swing1LimitDegrees=60
  swing2LimitDegrees=35
  swing2Motion=Locked
  twistLimitDegrees=45
  twistMotion=Locked
 }
 {bone1Name=dyn_leaf_top_rope_l, bone2Name=handle, swing1LimitDegrees=45, swing2LimitDegrees=20, twistLimitDegrees=45, twistMotion=Locked}
 {bone1Name=dyn_leaf_top_rope_r, bone2Name=handle, swing1LimitDegrees=40, swing2LimitDegrees=35, twistLimitDegrees=45, twistMotion=Locked}
]
